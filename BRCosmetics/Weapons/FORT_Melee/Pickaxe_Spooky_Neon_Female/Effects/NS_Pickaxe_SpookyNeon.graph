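# Niagara weapon trail effect for the Spooky Neon pickaxe, built from AnimTrail modules that spawn ribbons between bone or socket pairs.
# MainTrail draws a pink-violet ribbon, while HandleTrail adds a bright magenta ribbon on another socket pair.
# Two dark purple 0.25-second trail emitters, one with loop delay, layer ribbon and sprite renderers, using age-driven width, fade and dynamic material parameters.
# A randomized PinkTrail emitter is disabled.
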NiagaraSystem /BRCosmetics/Weapons/FORT_Melee/Pickaxe_Spooky_Neon_Female/Effects/NS_Pickaxe_SpookyNeon
dependencies {
 usedDataInterfaces [&NiagaraDataInterfaceCurve]
 usedDynamicInputs []
 usedModules [
  &self:SystemSpawnScript.SystemScriptSource.SystemScriptGraph.NiagaraNodeAssignment_0.SetVariables_#1
  &SystemState
 ]
}
emitters {
 FN_AnimTrail_V5 {
  emitterUpdateScript [
   &EmitterState(Life Cycle Mode=::NewEnumerator1, Loop Behavior=::NewEnumerator1, Loop Duration=0.25, UseLoopDelay=false)
   &FN_SpawnAnimTrail_v2(Spawn Interval=5) {enabled=false}
   /Game/Effects/Niagara/Modules/AnimTrail/FN_SpawnAnimTrail/FN_SpawnAnimTrail(Number of Anim Trails=1, SpawnRate=60)
  ]
  particleSpawnScript [
   /Game/Effects/Niagara/Modules/AnimTrail/FN_InitializeAnimTrail/FN_InitializeAnimTrailBoneIndecies(First Bone or Socket Index=1)
   /Game/Effects/Niagara/Modules/AnimTrail/FN_InitializeAnimTrail/FN_InitializeAnimTrail(
    AnimTrailWidth:NiagaraFloat=/Niagara/DynamicInputs/Math/LerpMultipleFloats(
     Float Count:ENiagara_MultipleLerpCount=::NewEnumerator0
     1:NiagaraFloat=1
     2:NiagaraFloat=3
     3:NiagaraFloat=1
     4:NiagaraFloat=?
     5:NiagaraFloat=?
     Alpha:NiagaraFloat=$Emitter.NormalizedLoopAge
    )
    Color=<rgba 0.0667 0.0275 0.1059 0.25>
    Lifetime:NiagaraFloat=&Lerp_Float(A=0.4, B=0.15, Alpha=$Emitter.NormalizedLoopAge)
   )
  ]
  particleUpdateScript [
   &FN_UpdateAnimTrail_v2()
   &ScaleRibbonWidth(Ribbon Width Scale=&FloatFromCurve(FloatCurve=DataInterface() {propertyValues {Curve {keys [{arriveTangent=-0.4183, interpMode=RCIM_Cubic, leaveTangent=-0.4183, tangentMode=RCTM_User, time=0, value=1}, <at 1 0.75>]}}}))
   &DynamicMaterialParameters(
    Index 0 Param 1=1
    Index 0 Param 2=&FloatFromCurve(FloatCurve=DataInterface() {propertyValues {Curve {keys [<at 0 1>,<at 1 0>]}}})
    Index 0 Param 3=0.1
    Index 0 Param 4=&FloatFromCurve(
     FloatCurve=DataInterface() {propertyValues {Curve {keys [
         {arriveTangent=3.4099, interpMode=RCIM_Cubic, leaveTangent=1.7024, tangentMode=RCTM_User, time=0, value=0}
         {arriveTangent=0.4146, interpMode=RCIM_Cubic, leaveTangent=0.4146, tangentMode=RCTM_User, time=1, value=1.25}
        ]}}}
     CurveIndex=$Emitter.NormalizedLoopAge
    )
    Module.Param0WriteEnabled=true
    Module.Param1WriteEnabled=true
    Module.Param2WriteEnabled=true
    Module.Param3WriteEnabled=true
   )
   &ScaleColor(
    Scale Alpha=&Lerp_Float(A=1, B=0.25, Alpha=$Particles.NormalizedAge)
    Scale RGB=&VectorFromFloat(Value=&Lerp_Float(A=1, B=0.1, Alpha=$Particles.NormalizedAge))
    ScaleA=true
    ScaleRGB=true
   )
  ]
  rendererClasses [&NiagaraRibbonRendererProperties, &NiagaraSpriteRendererProperties]
  renderers [
   {rendererClass=&NiagaraRibbonRendererProperties}
   {rendererClass=&NiagaraSpriteRendererProperties, rendererIndex=1}
   {rendererClass=&NiagaraRibbonRendererProperties, rendererIndex=2}
   {rendererClass=&NiagaraRibbonRendererProperties, rendererIndex=3}
  ]
 }
 FN_AnimTrail_V5001 {
  emitterUpdateScript [
   &EmitterState(Life Cycle Mode=::NewEnumerator1, Loop Behavior=::NewEnumerator1, Loop Delay=0.01, Loop Duration=0.25, UseLoopDelay=true)
   &FN_SpawnAnimTrail(Number of Anim Trails=1, SpawnRate=60)
  ]
  particleSpawnScript [
   &FN_InitializeAnimTrailBoneIndecies(First Bone or Socket Index=1)
   &FN_InitializeAnimTrail(
    AnimTrailWidth:NiagaraFloat=&LerpMultipleFloats(
     Float Count:ENiagara_MultipleLerpCount=::NewEnumerator0
     1:NiagaraFloat=2
     2:NiagaraFloat=3
     3:NiagaraFloat=1
     4:NiagaraFloat=?
     5:NiagaraFloat=?
     Alpha:NiagaraFloat=$Emitter.NormalizedLoopAge
    )
    Color=<rgba 0.0667 0.0275 0.1059 0.25>
    Lifetime=0.4
   )
  ]
  particleUpdateScript [
   &FN_UpdateAnimTrail_v2()
   &ScaleRibbonWidth(Ribbon Width Scale=&FloatFromCurve(FloatCurve=DataInterface() {propertyValues {Curve {keys [{arriveTangent=-0.4183, interpMode=RCIM_Cubic, leaveTangent=-0.4183, tangentMode=RCTM_User, time=0, value=1}, <at 1 0.75>]}}}))
   &DynamicMaterialParameters(
    Index 0 Param 1=1
    Index 0 Param 2=&FloatFromCurve(FloatCurve=DataInterface() {propertyValues {Curve {keys [<at 0 1>,<at 1 0>]}}}, CurveIndex=$Emitter.NormalizedLoopAge)
    Index 0 Param 3=0.1
    Index 0 Param 4=&FloatFromCurve(FloatCurve=DataInterface() {propertyValues {Curve {keys [
         {arriveTangent=3.4099, interpMode=RCIM_Cubic, leaveTangent=1.7024, tangentMode=RCTM_User, time=0, value=0}
         {arriveTangent=0.4146, interpMode=RCIM_Cubic, leaveTangent=0.4146, tangentMode=RCTM_User, time=1, value=1.25}
        ]}}})
    Module.Param0WriteEnabled=true
    Module.Param1WriteEnabled=true
    Module.Param2WriteEnabled=true
    Module.Param3WriteEnabled=true
   )
   &ScaleColor(
    Scale Alpha=&LerpMultipleFloats(
     Float Count:ENiagara_MultipleLerpCount=::NewEnumerator2
     1:NiagaraFloat=0
     2:NiagaraFloat=1
     3:NiagaraFloat=1
     4:NiagaraFloat=1
     5:NiagaraFloat=0
     Alpha:NiagaraFloat=$Particles.NormalizedAge
    )
    Scale RGB=&VectorFromFloat(Value=&Lerp_Float(A=1, B=0.1, Alpha=$Particles.NormalizedAge))
    ScaleA=true
    ScaleRGB=true
   )
  ]
  rendererClasses [&NiagaraRibbonRendererProperties, &NiagaraSpriteRendererProperties]
  renderers [
   {rendererClass=&NiagaraRibbonRendererProperties}
   {rendererClass=&NiagaraSpriteRendererProperties, rendererIndex=1}
   {rendererClass=&NiagaraRibbonRendererProperties, rendererIndex=2}
  ]
 }
 HandleTrail {
  emitterUpdateScript [
   &EmitterState()
   &FN_SpawnAnimTrail(Number of Anim Trails=1, SpawnRate=60)
  ]
  particleSpawnScript [
   &FN_InitializeAnimTrailBoneIndecies(First Bone or Socket Index=3, Second Bone or Socket Index=2)
   &FN_InitializeAnimTrailBoneIndecies(AnimTrail Selection=1, First Bone or Socket Index=4, Second Bone or Socket Index=5) {enabled=false, module=FN_InitializeAnimTrailBoneIndecies001}
   &FN_InitializeAnimTrail(
    AnimTrailWidth=2
    Color=<rgba 3 0 2.9446 1>
    Lifetime:NiagaraFloat=&Lerp_Float(A=0.075, B=0.005, Alpha=$System.NormalizedLoopAge)
   )
   &AddVelocity(Velocity=$System.TestVelocity) {enabled=false}
  ]
  particleUpdateScript [
   &SolveForcesAndVelocity(Clamp Velocity=false, Limit Acceleration=false)
   /Game/Effects/Niagara/Modules/AnimTrail/FN_UpdateAnimTrail/FN_UpdateAnimTrail()
   /Game/Effects/Niagara/Modules/Materials/Material_v2()
   &ScaleColor(
    Scale Alpha=&FloatFromCurve(FloatCurve=DataInterface() {propertyValues {Curve {keys [<at 0 0 RCIM_Cubic>,<at 0.15 1 RCIM_Cubic>,<at 1 0>]}}})
    ScaleA=true
    ScaleRGB=true
   )
  ]
  rendererClasses [&NiagaraRibbonRendererProperties]
  renderers [{rendererClass=&NiagaraRibbonRendererProperties}]
 }
 MainTrail {
  emitterUpdateScript [
   &EmitterState()
   &FN_SpawnAnimTrail(Number of Anim Trails=1, SpawnRate=60)
  ]
  particleSpawnScript [
   &FN_InitializeAnimTrailBoneIndecies(First Bone or Socket Index=1)
   &FN_InitializeAnimTrail(
    AnimTrailWidth=2
    Color=<rgba 0.9354 0.358 1 0.5>
    Lifetime:NiagaraFloat=&Lerp_Float(A=0.2, B=0, Alpha=$Emitter.NormalizedLoopAge)
   )
   &AddVelocity(Velocity=$System.TestVelocity) {enabled=false}
  ]
  particleUpdateScript [
   &SolveForcesAndVelocity(Clamp Velocity=false, Limit Acceleration=false)
   &FN_UpdateAnimTrail()
   &ScaleColor(
    Scale Alpha=&FloatFromCurve(FloatCurve=DataInterface() {propertyValues {Curve {keys [<at 0 0 RCIM_Cubic>,<at 0.15 1 RCIM_Cubic>,<at 1 0>]}}})
    ScaleA=true
    ScaleRGB=false
   )
  ]
  rendererClasses [&NiagaraRibbonRendererProperties]
  renderers [{rendererClass=&NiagaraRibbonRendererProperties}]
 }
 PinkTrail {
  bEnabled=false
  emitterUpdateScript [
   &EmitterState(Life Cycle Mode=::NewEnumerator1, Loop Behavior=::NewEnumerator1, Loop Duration=0.2, UseLoopDelay=false)
   /SetVariables(Emitter.Random:NiagaraInt32=&RandomRangeInt(Minimum=-2, Maximum=1)) {isSetParameters=true}
   &FN_SpawnAnimTrail(Number of Anim Trails:NiagaraInt32=$Emitter.Random, SpawnRate=60)
  ]
  particleSpawnScript [
   &FN_InitializeAnimTrailBoneIndecies(First Bone or Socket Index=1)
   &FN_InitializeAnimTrail(AnimTrailWidth=1, Color=<rgba 1 0 0.9815 1>, Lifetime=0.1)
   &AddVelocity(Velocity=$System.TestVelocity) {enabled=false}
  ]
  particleUpdateScript [
   &SolveForcesAndVelocity(Clamp Velocity=false, Limit Acceleration=false)
   &FN_UpdateAnimTrail()
   &Material_v2()
   &ScaleColor(
    Scale Alpha=&FloatFromCurve(FloatCurve=DataInterface() {propertyValues {Curve {keys [<at 0 0 RCIM_Cubic>,<at 0.15 1 RCIM_Cubic>,<at 1 0>]}}})
    ScaleA=true
    ScaleRGB=true
   )
  ]
  rendererClasses [&NiagaraRibbonRendererProperties]
  renderers [{rendererClass=&NiagaraRibbonRendererProperties}]
 }
}
summary {
 systemName=NS_Pickaxe_SpookyNeon
 emitters [
  {emitterName=MainTrail, rendererClasses [&NiagaraRibbonRendererProperties]}
  {bEnabled=false, emitterName=PinkTrail, rendererClasses [&NiagaraRibbonRendererProperties]}
  {emitterName=HandleTrail, rendererClasses [&NiagaraRibbonRendererProperties]}
  {
   emitterName=FN_AnimTrail_V5
   rendererClasses [&NiagaraRibbonRendererProperties, &NiagaraSpriteRendererProperties]
  }
  {
   emitterName=FN_AnimTrail_V5001
   rendererClasses [&NiagaraRibbonRendererProperties, &NiagaraSpriteRendererProperties]
  }
 ]
 userVariables [
  User.bIsFrontEndPreview {type=NiagaraBool}
  User.FEVec {type=Vector3f, description="", defaultValue {struct=Vector3f, value=<0 0 1000>}}
  User.PlayerSpeed {type=NiagaraFloat, defaultValue=0, description=""}
  User.RibbonIDOffset {type=NiagaraInt32}
  User.SkeletalMesh {
   type=NiagaraDataInterfaceSkeletalMesh
   defaultValue {
    struct=NiagaraExt_VariableValue_DataInterface
    value {dataInterface=&self:NiagaraDataInterfaceSkeletalMesh_0}
   }
  }
  User.SpawnRate {type=NiagaraFloat, defaultValue=360, description=""}
 ]
}
